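# Animation Blueprint for the Silent Tempo back bling: a clock that runs and gradually unfolds over time.
# An accumulated-delta timer wrapping every 360 seconds turns both hands, slides three panels and scales up the clock face, latching fully open after one cycle.
# Variant meta tags select reactive, held-closed (Stage2) or held-open (Stage3) via cached-pose blends, while the front end uses a per-update increment.
# GetGameTimeInSeconds-based clock logic remains unwired.
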AnimBlueprint /BRCosmetics/Accessories/FORT_Backpacks/Backpack_F_MED_SilentTempo/Meshes/F_MED_SilentTempo_Pack_AnimBP
target_skeleton=&/F_MED_SilentTempo_Pack_Skeleton
graphs [
 AnimGraph {nodes [
   n1 = Use cached pose 'StartPose'()
   n2 = Use cached pose 'ReactivityPose'()
   n3 = Get ReactiveOffClosed() -> (ReactiveOffClosed:bool)
   n4 = "Transform (Modify) Bone - Bone: pack"(
    ComponentPose=Local To Component(LocalPose=Blend Poses by bool(BlendPose_0=@n1, BlendPose_1=@n2, bActiveValue=@n3) {blendType=Linear})
    Translation="0,-10,1"
    Rotation="0,0,0"
   ) {
    translationMode=BMM_Additive
    translationSpace=BCS_BoneSpace
    alphaBoolBlend {blendOption=Linear}
    boneToModify {boneName=pack}
    translation {y=-10, z=1}
   }
   Output Pose(Result=Component To Local(ComponentPose=@n4)) {layerGroup=None}
   n5 = "Transform (Modify) Bone - Bone: dyn_arrow_small_jnt"!(
    ComponentPose=Local To Component(LocalPose=Input Pose())
    Alpha=1
    Translation="0,0,0"
    Rotation_Roll:float=Get hour() -> (hour:double)
    Rotation_Pitch:float=0
    Rotation_Yaw:float=0
    Scale="1,1,1"
   ) -> (Pose) {
    rotationMode=BMM_Additive
    rotationSpace=BCS_BoneSpace
    alphaBoolBlend {blendOption=Linear}
    boneToModify {boneName=dyn_arrow_small_jnt}
    rotation {roll=-60}
   }
   n6 = "Transform (Modify) Bone - Bone: dyn_arrow_big_jnt"!(
    ComponentPose=@n5
    Alpha=1
    Translation="0,0,0"
    Rotation_Roll:float=Get min() -> (min:double)
    Rotation_Pitch:float=0
    Rotation_Yaw:float=0
    Scale="1,1,1"
   ) -> (Pose) {rotationMode=BMM_Additive, rotationSpace=BCS_BoneSpace, alphaBoolBlend {blendOption=Linear}, boneToModify {boneName=dyn_arrow_big_jnt}}
   Save cached pose 'StartPose'(Pose=Component To Local(ComponentPose=@n6)) {cacheName=StartPose}
   n7 = "Transform (Modify) Bone - Bone: dyn_bpk_SilentTempo_ClockFace_jnt"(
    ComponentPose=Local To Component(LocalPose=Use cached pose 'ReactivityPreNumb'())
    Alpha=Get temp_for_numb() -> (temp_for_numb:double)
    Translation="0,0,0"
    Rotation="0,0,0"
    Scale="2.2,1,2.3"
   ) {
    scaleMode=BMM_Additive
    scaleSpace=BCS_BoneSpace
    alphaBoolBlend {blendOption=Linear}
    alphaScaleBiasClamp {interpSpeedDecreasing=0.5, interpSpeedIncreasing=0.5}
    boneToModify {boneName=dyn_bpk_SilentTempo_ClockFace_jnt}
    scale {x=2.2, z=2.3}
   }
   n8 = "Transform (Modify) Bone - Bone: dyn_bpk_SilentTempo_ClockFace_jnt"(
    ComponentPose=Local To Component(LocalPose=Use cached pose 'ReactivityPreNumb'())
    Alpha=Get temp_for_numb()
    Translation="0,0,0"
    Rotation="0,0,0"
    Scale="2.2,1,2.3"
   ) {
    scaleMode=BMM_Additive
    scaleSpace=BCS_BoneSpace
    alphaBoolBlend {blendOption=Linear}
    alphaScaleBiasClamp {bInterpResult=true, interpSpeedDecreasing=100, interpSpeedIncreasing=0.5}
    boneToModify {boneName=dyn_bpk_SilentTempo_ClockFace_jnt}
    scale {x=2.2, z=2.3}
   }
   Save cached pose 'ReactivityPose'(Pose=Blend Poses by bool(
     BlendPose_0=Component To Local(ComponentPose=@n7)
     BlendPose_1=Component To Local(ComponentPose=@n8)
     bActiveValue=Get ReactiveOffOpen() -> (ReactiveOffOpen:bool)
    ) {blendType=Linear}) {cacheName=ReactivityPose}
   n9 = Get translate_and_scale() -> (translate_and_scale:double)
   n10 = "Transform (Modify) Bone - Bone: dyn_bpk_SilentTempo_upper_jnt"(ComponentPose=Local To Component(LocalPose=Use cached pose 'StartPose'()), Alpha=@n9, Translation="-3,0,0", Rotation="0,0,0", Scale="1,1,1") {
    translationMode=BMM_Additive
    translationSpace=BCS_BoneSpace
    alphaBoolBlend {blendOption=Linear}
    alphaScaleBiasClamp {interpSpeedDecreasing=1, interpSpeedIncreasing=1}
    boneToModify {boneName=dyn_bpk_SilentTempo_upper_jnt}
    translation {x=-3}
   }
   n11 = "Transform (Modify) Bone - Bone: dyn_bpk_SilentTempo_r_jnt"(ComponentPose=@n10, Alpha=@n9, Translation="-3,0,0", Rotation="0,0,0", Scale="1,1,1") {
    translationMode=BMM_Additive
    translationSpace=BCS_BoneSpace
    alphaBoolBlend {blendOption=Linear}
    boneToModify {boneName=dyn_bpk_SilentTempo_r_jnt}
    translation {x=-3}
   }
   n12 = "Transform (Modify) Bone - Bone: dyn_bpk_SilentTempo_l_jnt"(ComponentPose=@n11, Alpha=@n9, Translation="-3,0,0", Rotation="0,0,0", Scale="1,1,1") {
    translationMode=BMM_Additive
    translationSpace=BCS_BoneSpace
    alphaBoolBlend {blendOption=Linear}
    boneToModify {boneName=dyn_bpk_SilentTempo_l_jnt}
    translation {x=-3}
   }
   Save cached pose 'ReactivityPreNumb'(Pose=Component To Local(ComponentPose=@n12)) {cacheName=ReactivityPreNumb}
  ]}
 EventGraph {nodes [
   n1 = Event BlueprintUpdateAnimation()
   > n2 = Sequence!(execute) -> (then_0, then_1, then_2:Exec, then_3:Exec, then_4:Exec)
   n3 = Branch(execute=@n2.then_1, Condition="float >= float"(A=Get gameTimeADJ() -> (gameTimeADJ:double), B=360) {is=title K2Node_CallFunction})
   > Set gameTimeADJ(gameTimeADJ:double=float - float(A=Get gameTimeADJ(), B=360) {is=title K2Node_CallFunction}) -> (Output_Get:double)
   > Do Once()
   > Set deployed(deployed:bool=true) -> (Output_Get:bool)
   Set gameTimeADJ(execute=@n3.else, gameTimeADJ=float + float(A=@n1.DeltaTimeX, B=Get gameTimeADJ()))
   Set ReactiveOffClosed(
    execute=@n2.then_0
    ReactiveOffClosed:bool=BackpackItemDefinitionHasAnyMetaTag!(Tags:Array of Gameplay Tag Structures=Make Array("[0]"=(TagName="Cosmetics.Variant.Property.Stage2")), bRequiresEquipped:bool=false) -> (ReturnValue:bool)
   ) -> (Output_Get:bool)
   > Set ReactiveOffOpen(ReactiveOffOpen:bool=BackpackItemDefinitionHasAnyMetaTag(Tags=Make Array("[0]"=(TagName="Cosmetics.Variant.Property.Stage3")), bRequiresEquipped=false)) -> (Output_Get:bool)
   n4 = Get ReactiveOffOpen()
   n5 = Get ReactiveOffClosed()
   > n6 = Branch(Condition=AND Boolean!(A=Get bIsInFrontEnd(), B=NOT Boolean(A=@n5), C:bool=NOT Boolean(A=@n4)) -> (ReturnValue))
   > n7 = Branch(Condition=Get FE Looped() -> (FE Looped:bool))
   Set translate_and_scale(execute=@n7.else, translate_and_scale:double=float + float(A=Get translate_and_scale(), B=0.0007)) -> (Output_Get:double)
   > Set temp_for_numb(temp_for_numb:double=float + float(A=Get temp_for_numb(), B=0.0007)) -> (Output_Get:double)
   > Branch(Condition="float >= float"(A=Get translate_and_scale(), B=1) {is=title K2Node_PromotableOperator})
   > Set FE Looped(FE Looped:bool=true) -> (Output_Get:bool)
   n8 = Get gameTimeADJ()
   Set translate_and_scale(execute=@n6.else, translate_and_scale=MapRangeClamped(Value=@n8, InRangeA=0, InRangeB=300))
   > Set temp_for_numb(temp_for_numb=MapRangeClamped(Value=@n8, InRangeA=100, InRangeB=200))
   > n9 = Sequence!(execute) -> (then_0, then_1)
   > Branch(Condition=Get ReactiveOffOpen())
   > Set translate_and_scale(translate_and_scale=1)
   > Set temp_for_numb(temp_for_numb=1)
   Branch(execute=@n9.then_1, Condition=Get deployed() -> (deployed:bool))
   > Set translate_and_scale(translate_and_scale=1)
   > Set temp_for_numb(temp_for_numb=1)
   Set hour(execute=@n2.then_3, hour:double=MapRangeClamped(Value=Get gameTimeADJ(), InRangeA=0, InRangeB=360, OutRangeA=-5, OutRangeB=-365)) -> (Output_Get:double)
   Set min(execute=@n2.then_2, min:double=MapRangeClamped(Value=Get gameTimeADJ(), InRangeA=0, InRangeB=360, OutRangeA=-30, OutRangeB=-4320)) -> (Output_Get:double)
   Reroute Node()
   n10 = GetGameTimeInSeconds()
   n11 = int + int(A=Get for_hour_index() -> (for_hour_index:int=0))
   > n12 = Branch(Condition="float >= float"(A=@n10, B=int * float(A:int=360, B:int=@n11) -> (ReturnValue=~) {is=title K2Node_PromotableOperator}))
   > Set for_hour_index(for_hour_index:int=int + int(A=Get for_hour_index() -> (for_hour_index=0))) -> (Output_Get:int=0)
   n13 = int * int(A=Get for_hour_index() -> (for_hour_index=0), B=360)
   Set hour(
    execute=@n12.else
    hour=MapRangeClamped(Value=float - float(A=@n10, B:int=@n13) {is=title K2Node_PromotableOperator}, InRangeA=0, InRangeB=360, OutRangeA=-5, OutRangeB=-365)
   )
   Reroute Node()
   > Set translate_and_scale(translate_and_scale=MapRangeClamped(Value=@n10, InRangeA=0, InRangeB=300))
   > Set temp_for_numb(temp_for_numb=MapRangeClamped(Value=@n10, InRangeA=100, InRangeB=200))
   > Branch(Condition=Get ReactiveOffOpen())
   > Set translate_and_scale(translate_and_scale=1)
   > Set temp_for_numb(temp_for_numb=1)
   Reroute Node()
   n14 = int + int(A=Get Last Index() -> (Last Index:int=0))
   > n15 = Branch(Condition="float >= float"(A=@n10, B=int * float(A=60, B=@n14) -> (ReturnValue=~)))
   > Set Last Index(Last Index:int=int + int(A=Get Last Index() -> (Last Index=0))) -> (Output_Get:int=0)
   n16 = int * int(A=Get Last Index() -> (Last Index=0), B=60)
   Set min(execute=@n15.else, min=MapRangeClamped(Value=float - float(A=@n10, B=@n16), InRangeA=0, InRangeB=60, OutRangeB=-720))
   Set translate_and_scale(translate_and_scale=0)
   > Set temp_for_numb(temp_for_numb=0)
   > Delay(Duration=2)
   > Set FE Looped(FE Looped=false)
   Delay(Duration=0.1)
  ]}
]
variables [
 deployed {type=boolean}
 FE Looped {type=boolean}
 for_hour_index {type=integer}
 gameTimeADJ {type=number}
 gameTimeMOD {type=integer}
 gameTimeWRAP {type=number}
 hour {type=number}
 Last Index {type=integer}
 min {type=number}
 ReactiveOffClosed {type=boolean}
 ReactiveOffOpen {type=boolean}
 temp_for_numb {type=number}
 translate_and_scale {type=number}
]
